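# Post-process style Anim Blueprint for the M_MED_Sleek_Orange character body, adding deformation and secondary motion to the copied parent-mesh pose.
# Twist and sculpter Control Rigs feed additive jacket and collar corrections driven by thigh and arm rotations, then jacket and necklace RigidBody simulations.
# Updates compute those alphas and reduce twist-rig weight during Shaolin emote montages.
# They also set a GravityOverride the AnimGraph never reads; both simulations use fixed gravity.
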AnimBlueprint /BRCosmetics/Characters/Player/Male/Medium/Bodies/M_MED_Sleek_Orange/Meshes/M_MED_Sleek_Orange_AnimBP
target_skeleton=&/M_MED_Sleek_Orange_Skeleton
graphs [
 AnimGraph {nodes [
   n1 = Control Rig(Source=Copy Pose From Mesh() {bUseAttachedParent=true}, Alpha=Get ShaolinEmoteAlpha() -> (ShaolinEmoteAlpha:double)) {
    alphaBoolBlend {blendOption=Linear}
    controlRigAssetReference {blueprintRigClass=/BRCosmetics/Characters/Player/Male/Medium/Bodies/M_MED_Sleek/Meshes/M_MED_Sleek_Deform_CtrlRig_twists.M_MED_Sleek_Deform_CtrlRig_twists_C}
   }
   n2 = Control Rig!(
    Source=@n1
    Alpha
    KneeRot_fr_l_114deg:Vector
    KneeTrans_fr_l_114deg:Vector="0.258,-3.591,0.009"
    KneeScale_fr_l_114deg:Vector="1,1,1"
    KneeScale_fr_l_130deg:Vector="1,1,1"
    KneeRot_fr_l_130deg:Vector="0,0,-51.2"
    KneeTrans_fr_l_130deg:Vector="0.258,-3.591,0.009"
   ) -> (Pose) {
    alphaBoolBlend {blendOption=Linear}
    controlRigAssetReference {blueprintRigClass=/BRCosmetics/Characters/Player/Male/Medium/Bodies/M_MED_Sleek/Meshes/M_MED_Sleek_Deform_CtrlRig_sculpters.M_MED_Sleek_Deform_CtrlRig_sculpters_C}
   }
   n3 = Get thigh_up_l() -> (thigh_up_l:double)
   n4 = "Transform (Modify) Bone - Bone: dyn_jacket_l_fr_MSTR"(ComponentPose=Local To Component(LocalPose=@n2), Alpha=@n3, Translation="0,-3,3", Rotation="25,-35,-60", Scale="1,1,1") {
    rotationMode=BMM_Additive
    rotationSpace=BCS_BoneSpace
    translationMode=BMM_Additive
    translationSpace=BCS_BoneSpace
    alphaBoolBlend {blendOption=Linear}
    boneToModify {boneName=dyn_jacket_l_fr_MSTR}
    rotation=<rot 25 -35 -60>
    translation {y=-3, z=3}
   }
   n5 = "Transform (Modify) Bone - Bone: dyn_jacket_l_MSTR"(ComponentPose=@n4, Alpha=@n3, Translation="0,0,2", Rotation="15,0,0", Scale="1,1,1") {
    rotationMode=BMM_Additive
    rotationSpace=BCS_BoneSpace
    translationMode=BMM_Additive
    translationSpace=BCS_BoneSpace
    alphaBoolBlend {blendOption=Linear}
    boneToModify {boneName=dyn_jacket_l_MSTR}
    rotation {pitch=15}
    translation {z=2}
   }
   n6 = Get thigh_up_r() -> (thigh_up_r:double)
   n7 = "Transform (Modify) Bone - Bone: dyn_jacket_r_fr_MSTR"(ComponentPose=@n5, Alpha=@n6, Translation="0,3,3", Rotation="25,35,60", Scale="1,1,1") {
    rotationMode=BMM_Additive
    rotationSpace=BCS_BoneSpace
    translationMode=BMM_Additive
    translationSpace=BCS_BoneSpace
    alphaBoolBlend {blendOption=Linear}
    boneToModify {boneName=dyn_jacket_r_fr_MSTR}
    rotation=<rot 25 35 60>
    translation {y=3, z=3}
   }
   n8 = "Transform (Modify) Bone - Bone: dyn_jacket_r_MSTR"(ComponentPose=@n7, Alpha=@n6, Translation="0,0,0", Rotation="15,0,0", Scale="1,1,1") {
    rotationMode=BMM_Additive
    rotationSpace=BCS_BoneSpace
    translationMode=BMM_Additive
    translationSpace=BCS_BoneSpace
    alphaBoolBlend {blendOption=Linear}
    boneToModify {boneName=dyn_jacket_r_MSTR}
    rotation {pitch=15}
   }
   n9 = "Transform (Modify) Bone - Bone: def_collar_r"(ComponentPose=@n8, Alpha=Get arm_rot_up_r() -> (arm_rot_up_r:double), Translation="0,0,0", Rotation="0.000,-25.000,-5.000", Scale="1,1,1") {
    rotationMode=BMM_Additive
    rotationSpace=BCS_BoneSpace
    scaleMode=BMM_Additive
    scaleSpace=BCS_BoneSpace
    translationMode=BMM_Additive
    translationSpace=BCS_BoneSpace
    alphaBoolBlend {blendOption=Linear}
    boneToModify {boneName=def_collar_r}
    rotation {roll=-5, yaw=-24.9998}
   }
   n10 = "Transform (Modify) Bone - Bone: def_collar_l"(ComponentPose=@n9, Alpha=Get arm_rot_up_l() -> (arm_rot_up_l:double), Translation="0,0,0", Rotation="0.000,10,0", Scale="0.85,1.000,1.000") {
    rotationMode=BMM_Additive
    rotationSpace=BCS_BoneSpace
    scaleMode=BMM_Additive
    scaleSpace=BCS_BoneSpace
    translationMode=BMM_Additive
    translationSpace=BCS_BoneSpace
    alphaBoolBlend {blendOption=Linear}
    boneToModify {boneName=def_collar_l}
    rotation {yaw=10}
    scale {x=0.85}
   }
   n11 = RigidBody!(ComponentPose=@n10, Alpha, OverrideWorldGravity:Vector="0,0,-980", ExternalForce) -> (Pose) {
    bClampLinearTranslationLimitToRefPose=true
    bOverrideWorldGravity=true
    overridePhysicsAsset=&/M_MED_Sleek_Orange_jacket_dyn
    simulationSpace=BaseBoneSpace
    alphaBoolBlend {blendOption=Linear}
    baseBoneRef {boneName=spine_02}
    componentAppliedLinearAccClamp=<3000 3000 3000>
    componentLinearAccScale {x=0.4}
    componentLinearVelScale {x=0.4}
    overrideWorldGravity {z=-980}
    simSpaceSettings {
     maxAngularAcceleration=200
     maxAngularVelocity=5
     maxLinearAcceleration=1000
     maxLinearVelocity=200
     velocityScaleZ=0
     worldAlpha=1
     externalLinearDragV=<1 1>
    }
   }
   n12 = RigidBody(ComponentPose=@n11, OverrideWorldGravity="0,0,-980") {
    bClampLinearTranslationLimitToRefPose=true
    bOverrideWorldGravity=true
    overridePhysicsAsset=&/M_MED_Sleek_Orange_dyn_necklace
    simulationSpace=BaseBoneSpace
    alphaBoolBlend {blendOption=Linear}
    baseBoneRef {boneName=spine_05}
    componentAppliedLinearAccClamp=<2000 2000 1000>
    componentLinearAccScale {x=0.4}
    componentLinearVelScale {x=0.4}
    overrideWorldGravity {z=-980}
    simSpaceSettings {maxAngularAcceleration=100, maxAngularVelocity=5, maxLinearAcceleration=300, maxLinearVelocity=50, velocityScaleZ=0, worldAlpha=1}
   }
   Output Pose(Result=Component To Local(ComponentPose=@n12)) {layerGroup=None}
  ]}
 EventGraph {nodes [
   Event BlueprintUpdateAnimation()
   > n1 = Sequence!(execute) -> (then_0, then_1, then_2:Exec)
   n2 = BreakTransform(InTransform=GetDeltaTransformFromRefPose(self=GetOwnerSkeletalMeshComponent(), BoneName=thigh_l, BaseName=pelvis))
   > Set thigh_up_l(thigh_up_l:double=MapRangeClamped(Value=BreakRotator(InRot=@n2.Rotation).Yaw, InRangeA=-20, InRangeB=-100)) -> (Output_Get:double)
   n3 = BreakTransform(InTransform=GetDeltaTransformFromRefPose(self=GetOwnerSkeletalMeshComponent(), BoneName=thigh_r, BaseName=pelvis))
   > Set thigh_up_r(thigh_up_r:double=MapRangeClamped(Value=BreakRotator(InRot=@n3.Rotation).Yaw, InRangeA=-20, InRangeB=-100)) -> (Output_Get:double)
   n4 = GetOwnerSkeletalMeshComponent()
   n5 = BreakRotator(InRot=BreakTransform(InTransform=GetDeltaTransformFromRefPose(self=@n4, BoneName=upperarm_r, BaseName=spine_05)).Rotation)
   > Set arm_rot_up_r(arm_rot_up_r:double=MapRangeClamped(Value="Absolute (Float)"(A=@n5.Yaw), InRangeA=20, InRangeB=100)) -> (Output_Get:double)
   n6 = GetOwnerSkeletalMeshComponent()
   n7 = BreakRotator(InRot=BreakTransform(InTransform=GetDeltaTransformFromRefPose(self=@n6, BoneName=upperarm_l, BaseName=spine_05)).Rotation)
   > Set arm_rot_up_l(arm_rot_up_l:double=MapRangeClamped(Value="Absolute (Float)"(A=@n7.Yaw), InRangeA=20, InRangeB=120)) -> (Output_Get:double)
   n8 = Gravity Override State Function DEPRECATED(
    execute=@n1.then_1
    CharacterPartAnimInstance=Self-Reference()
    GameTimeInSeconds=GetGameTimeInSeconds()
    GravityOverrideParams=Get Gravity Override Params()
   )
   > Set GravityOverride(GravityOverride=@n8)
   > n9 = Branch(Condition=Get bIsDBNO())
   > Set GravityOverride(GravityOverride="0,0,-980")
   Set GravityOverride(execute=@n9.else, GravityOverride="0, 0, 0")
   n10 = GetAnimInstance(self=GetOwnerSkeletalMeshComponent())
   Is Valid(exec=@n1.then_2, InputObject=@n10)
   n11 = Montage_IsActive!(
    self:Anim Instance Object Reference=@n10
    Montage:Anim Montage Object Reference=/BRCosmetics/Animation/Game/MainPlayer/Emotes/ShaolinSipUp/Emote_ShaolinSipUp_CMM_M
   ) -> (ReturnValue:bool)
   > n12 = Branch(Condition=OR Boolean(A=@n11, B=Montage_IsActive(self=@n10, Montage=/BRCosmetics/Animation/Game/MainPlayer/Emotes/Shaolin/Emote_Shaolin_M)))
   > Set ShaolinEmoteAlpha(ShaolinEmoteAlpha:double=SelectFloat(A=0.2, B=0.4, bPickA=@n11)) -> (Output_Get:double)
   Set ShaolinEmoteAlpha(execute=@n12.else, ShaolinEmoteAlpha=1)
  ]}
]
variables [
 arm_rot_up_l {type=number}
 arm_rot_up_r {type=number}
 Gravity Override Params {
  type=GravityOverrideParamsStruct
  default {
   skydive_FloatingDown {gravityOverride=<0 -700 100>, windAmplitude=<200 -500 200>, windOffset=<0 -600 200>}
   skydive_GoingUp {gravityOverride=<0 -300 500>, windAmplitude=<200 -500 300>}
   skydive_Parachuting {gravityOverride=<0 -500 -980>, windAmplitude=<200 -300 400>}
  }
 }
 GravityOverride {type=Vector}
 ShaolinEmoteAlpha {type=number}
 thigh_up_l {type=number}
 thigh_up_r {type=number}
]
